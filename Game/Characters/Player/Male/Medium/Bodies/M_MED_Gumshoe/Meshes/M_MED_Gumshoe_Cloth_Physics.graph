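# Physics asset for the cloth of the medium male Gumshoe player body, simulating dynamic leg-cloth bones and a lightbox accessory.
# Capsules on pelvis, thighs and calves form the body, joined by angularly free constraints.
# Three-link front, outer and rear chains hang from each thigh with limited swing and twist, with the lightbox on the pelvis.
# Free-rotation constraints cross-link front and outer chain tips on each side.
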PhysicsAsset /Game/Characters/Player/Male/Medium/Bodies/M_MED_Gumshoe/Meshes/M_MED_Gumshoe_Cloth_Physics
bodies {
 calf_l [{shapeType=Capsule, length=48.6973, radius=12.4498, center=<-23.5048 -1.1257 -0.0099>, rotation=<rot -0.6019 -91.9574 -88.736>}]
 calf_r [{shapeType=Capsule, length=53.3766, radius=11.2739, center=<28.5545 0.5205 0.2376>, rotation=<rot -0.1465 89.0026 -88.8377>}]
 dyn_frontLeg_l_01 [{shapeType=Capsule, length=39.1364, radius=6.1955, shapeName=dyn_frontLeg_l_01_capsule, center=<7.6339 3 0>}]
 dyn_frontLeg_l_02 [{shapeType=Capsule, length=39.1364, radius=6.1955, shapeName=dyn_frontLeg_l_02_capsule, center=<7.6339 3 0>}]
 dyn_frontLeg_l_03 [{shapeType=Capsule, length=39.1364, radius=6.1955, shapeName=dyn_frontLeg_l_03_capsule, center=<7.6339 3 0>}]
 dyn_frontLeg_r_01 [{shapeType=Capsule, length=21.704, radius=1.7995, shapeName=dyn_frontLeg_r_01_capsule, center=<7.6339 3 0>}]
 dyn_frontLeg_r_02 [{shapeType=Capsule, length=32.9504, radius=5.9603, shapeName=dyn_frontLeg_r_02_capsule, center=<7.6339 3 0>}]
 dyn_frontLeg_r_03 [{shapeType=Capsule, length=32.9504, radius=5.9603, shapeName=dyn_frontLeg_r_03_capsule, center=<7.6339 3 0>}]
 dyn_l_mstr [{shapeType=Capsule, length=0.5, radius=0.5, shapeName=dyn_l_mstr_capsule}]
 dyn_lightbox [{shapeType=Capsule, length=7.9507, radius=0.9693, shapeName=dyn_lightbox_capsule, center=<6.5363 0 0>, rotation=<rot 90 0 0>}]
 dyn_outerLeg_l_01 [{shapeType=Capsule, length=25.4972, radius=6.1955, shapeName=dyn_outerLeg_l_01_capsule, center=<7.6339 0 3>, rotation=<rot 0 0 90>}]
 dyn_outerLeg_l_02 [{shapeType=Capsule, length=25.4972, radius=6.1955, shapeName=dyn_outerLeg_l_02_capsule, center=<7.6339 0 3>, rotation=<rot 0 0 90>}]
 dyn_outerLeg_l_03 [{shapeType=Capsule, length=25.4972, radius=6.1955, shapeName=dyn_outerLeg_l_03_capsule, center=<7.6339 0 3>, rotation=<rot 0 0 90>}]
 dyn_outerLeg_r_01 [{shapeType=Capsule, length=27.0731, radius=5.9603, shapeName=dyn_outerLeg_r_01_capsule, center=<7.6339 0 -5.5>, rotation=<rot 0 0 90>}]
 dyn_outerLeg_r_02 [{shapeType=Capsule, length=27.0731, radius=5.9603, shapeName=dyn_outerLeg_r_02_capsule, center=<7.6339 0 -3>, rotation=<rot 0 0 90>}]
 dyn_outerLeg_r_03 [{shapeType=Capsule, length=27.0731, radius=5.9603, shapeName=dyn_outerLeg_r_03_capsule, center=<7.6339 0 -3>, rotation=<rot 0 0 90>}]
 dyn_r_mstr [{shapeType=Capsule, length=0.5, radius=0.5, shapeName=dyn_r_mstr_capsule}]
 dyn_rearLeg_l_01 [{shapeType=Capsule, length=12.2988, radius=8.081, shapeName=dyn_rearLeg_l_01_capsule, center=<7.6339 -3 0>}]
 dyn_rearLeg_l_02 [{shapeType=Capsule, length=12.2988, radius=8.081, shapeName=dyn_rearLeg_l_02_capsule, center=<7.6339 -3 0>}]
 dyn_rearLeg_l_03 [{shapeType=Capsule, length=12.2988, radius=8.081, shapeName=dyn_rearLeg_l_03_capsule, center=<7.6339 -3 0>}]
 dyn_rearLeg_r_01 [{shapeType=Capsule, length=13.0634, radius=7.8466, shapeName=dyn_rearLeg_r_01_capsule, center=<7.6339 -3 0>}]
 dyn_rearLeg_r_02 [{shapeType=Capsule, length=13.0634, radius=7.8466, shapeName=dyn_rearLeg_r_02_capsule, center=<7.6339 -3 0>}]
 dyn_rearLeg_r_03 [{shapeType=Capsule, length=13.0634, radius=7.8466, shapeName=dyn_rearLeg_r_03_capsule, center=<7.6339 -3 0>}]
 pelvis [{shapeType=Capsule, length=25.7786, radius=12.4856, center=<9.0316 -0.0002 0>, rotation=<rot 90 -0.0049 -0.0049>}]
 thigh_l [{shapeType=Capsule, length=29.2756, radius=11.1933, center=<-22.2715 0.3445 2.0131>, rotation=<rot -89.3829 91.7043 -89.1787>}]
 thigh_r [{shapeType=Capsule, length=39.2398, radius=11.8301, center=<24.8817 -0.7692 -1.2356>, rotation=<rot -90 179.7016 2.8482>}]
}
constraints [
 {
  bone1Name=calf_l
  bone2Name=thigh_l
  swing1LimitDegrees=45
  swing1Motion=Free
  swing2LimitDegrees=45
  swing2Motion=Free
  twistLimitDegrees=45
  twistMotion=Free
 }
 {
  bone1Name=thigh_l
  bone2Name=pelvis
  swing1LimitDegrees=45
  swing1Motion=Free
  swing2LimitDegrees=45
  swing2Motion=Free
  twistLimitDegrees=45
  twistMotion=Free
 }
 {
  bone1Name=thigh_r
  bone2Name=pelvis
  swing1LimitDegrees=45
  swing1Motion=Free
  swing2LimitDegrees=45
  swing2Motion=Free
  twistLimitDegrees=45
  twistMotion=Free
 }
 {
  bone1Name=calf_r
  bone2Name=thigh_r
  swing1LimitDegrees=45
  swing1Motion=Free
  swing2LimitDegrees=45
  swing2Motion=Free
  twistLimitDegrees=45
  twistMotion=Free
 }
 {bone1Name=dyn_frontLeg_r_01, bone2Name=thigh_r, swing1LimitDegrees=60, swing2LimitDegrees=60, twistLimitDegrees=30}
 {bone1Name=dyn_frontLeg_r_02, bone2Name=dyn_frontLeg_r_01, swing1LimitDegrees=30, swing2LimitDegrees=30, twistLimitDegrees=30}
 {bone1Name=dyn_frontLeg_r_03, bone2Name=dyn_frontLeg_r_02, swing1LimitDegrees=30, swing2LimitDegrees=30, twistLimitDegrees=30}
 {bone1Name=dyn_outerLeg_r_01, bone2Name=thigh_r, swing1LimitDegrees=60, swing2LimitDegrees=60, twistLimitDegrees=30}
 {bone1Name=dyn_outerLeg_r_02, bone2Name=dyn_outerLeg_r_01, swing1LimitDegrees=60, swing2LimitDegrees=60, twistLimitDegrees=30}
 {bone1Name=dyn_outerLeg_r_03, bone2Name=dyn_outerLeg_r_02, swing1LimitDegrees=60, swing2LimitDegrees=60, twistLimitDegrees=30}
 {bone1Name=dyn_r_mstr, bone2Name=thigh_r, swing1LimitDegrees=30, swing2LimitDegrees=30, twistLimitDegrees=5}
 {bone1Name=dyn_rearLeg_r_01, bone2Name=thigh_r, swing1LimitDegrees=60, swing2LimitDegrees=60, twistLimitDegrees=30}
 {bone1Name=dyn_rearLeg_r_02, bone2Name=dyn_rearLeg_r_01, swing1LimitDegrees=60, swing2LimitDegrees=60, twistLimitDegrees=30}
 {bone1Name=dyn_rearLeg_r_03, bone2Name=dyn_rearLeg_r_02, swing1LimitDegrees=60, swing2LimitDegrees=60, twistLimitDegrees=30}
 {bone1Name=dyn_rearLeg_l_03, bone2Name=dyn_rearLeg_l_02, swing1LimitDegrees=60, swing2LimitDegrees=60, twistLimitDegrees=30}
 {bone1Name=dyn_outerLeg_l_01, bone2Name=thigh_l, swing1LimitDegrees=60, swing2LimitDegrees=60, twistLimitDegrees=30}
 {bone1Name=dyn_outerLeg_l_02, bone2Name=dyn_outerLeg_l_01, swing1LimitDegrees=60, swing2LimitDegrees=60, twistLimitDegrees=30}
 {bone1Name=dyn_outerLeg_l_03, bone2Name=dyn_outerLeg_l_02, swing1LimitDegrees=60, swing2LimitDegrees=60, twistLimitDegrees=30}
 {bone1Name=dyn_rearLeg_l_01, bone2Name=thigh_l, swing1LimitDegrees=60, swing2LimitDegrees=60, twistLimitDegrees=30}
 {bone1Name=dyn_rearLeg_l_02, bone2Name=dyn_rearLeg_l_01, swing1LimitDegrees=60, swing2LimitDegrees=60, twistLimitDegrees=30}
 {bone1Name=dyn_frontLeg_l_03, bone2Name=dyn_frontLeg_l_02, swing1LimitDegrees=45, swing2LimitDegrees=45, twistLimitDegrees=45}
 {bone1Name=dyn_frontLeg_l_02, bone2Name=dyn_frontLeg_l_01, swing1LimitDegrees=45, swing2LimitDegrees=45, twistLimitDegrees=45}
 {bone1Name=dyn_frontLeg_l_01, bone2Name=thigh_l, swing1LimitDegrees=45, swing2LimitDegrees=45, twistLimitDegrees=45}
 {bone1Name=dyn_lightbox, bone2Name=pelvis, swing1LimitDegrees=30, swing2LimitDegrees=30, twistLimitDegrees=5}
 {
  bone1Name=dyn_l_mstr
  bone2Name=thigh_l
  swing1LimitDegrees=45
  swing1Motion=Free
  swing2LimitDegrees=45
  swing2Motion=Free
  twistLimitDegrees=45
  twistMotion=Free
 }
 {
  bone1Name=dyn_frontLeg_r_03
  bone2Name=dyn_outerLeg_r_03
  swing1LimitDegrees=45
  swing1Motion=Free
  swing2LimitDegrees=45
  swing2Motion=Free
  twistLimitDegrees=45
  twistMotion=Free
 }
 {
  bone1Name=dyn_outerLeg_l_03
  bone2Name=dyn_frontLeg_l_03
  swing1LimitDegrees=45
  swing1Motion=Free
  swing2LimitDegrees=45
  swing2Motion=Free
  twistLimitDegrees=45
  twistMotion=Free
 }
]
