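# Animation Blueprint for the Gumshoe player body, copying the attached parent mesh's pose and adding secondary motion to front, outer and rear leg dyn bones, plus lightbox and collar.
# Flap nodes map thigh rotation onto leg tip rotation, and calf_rot_r blends in right-side offsets.
# Chained AnimDynamics springs use sphere limits as pelvis and leg collision.
# Later passes add corrective offsets, calf-driven tip scaling, and a weighted constraint feeding lightbox simulation.
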AnimBlueprint /Game/Characters/Player/Male/Medium/Bodies/M_MED_Gumshoe/Meshes/M_MED_Gumshoe_Export_Skeleton_AnimBlueprint
target_skeleton=&/M_MED_Gumshoe_Export_Skeleton
graphs [
 AnimGraph {nodes [
   n1 = Get MeshToCopy() -> (MeshToCopy:Skeletal Mesh Component Object Reference)
   n2 = Flap Node - Bone: dyn_rearLeg_r_03!(
    ComponentPose:Component Space Pose Link Structure=Local To Component(LocalPose=Copy Pose From Mesh(SourceMeshComponent=@n1) {bCopyCurves=true, bUseAttachedParent=true})
    Alpha:float=1
   ) -> (Pose:Component Space Pose Link Structure) {
    is=title FortAnimGraphNode_Flap
    targetComponent=RotationZ
    alphaBoolBlend {blendOption=Linear}
    sourceDrives [{
      bUseRange=true
      drivingCurve=null
      multiplier=1
      rangeMax=0.35
      rangeMin=-0.35
      remappedMax=0.35
      remappedMin=-0.35
      sourceComponent=RotationZ
      useQuaternion=true
      sourceBone {boneName=thigh_r}
     }]
    targetBone {boneName=dyn_rearLeg_r_03}
   }
   n3 = Flap Node - Bone: dyn_outerLeg_r_03!(ComponentPose:Component Space Pose Link Structure=@n2, Alpha:float=1) -> (Pose:Component Space Pose Link Structure) {
    is=title FortAnimGraphNode_Flap
    targetComponent=RotationZ
    alphaBoolBlend {blendOption=Linear}
    sourceDrives [{
      bUseRange=true
      drivingCurve=null
      multiplier=1
      rangeMax=0.35
      rangeMin=-0.35
      remappedMax=0.35
      remappedMin=-0.35
      sourceComponent=RotationZ
      useQuaternion=true
      sourceBone {boneName=thigh_r}
     }]
    targetBone {boneName=dyn_outerLeg_r_03}
   }
   n4 = Flap Node - Bone: dyn_frontLeg_r_03!(ComponentPose:Component Space Pose Link Structure=@n3, Alpha:float=1) -> (Pose:Component Space Pose Link Structure) {
    is=title FortAnimGraphNode_Flap
    targetComponent=RotationZ
    alphaBoolBlend {blendOption=Linear}
    sourceDrives [{
      bUseRange=true
      drivingCurve=null
      multiplier=1
      rangeMax=0.1725
      rangeMin=-0.1725
      remappedMax=0.1725
      remappedMin=-0.1725
      sourceComponent=RotationZ
      useQuaternion=true
      sourceBone {boneName=thigh_r}
     }]
    targetBone {boneName=dyn_frontLeg_r_03}
   }
   n5 = Flap Node - Bone: dyn_rearLeg_l_03!(ComponentPose:Component Space Pose Link Structure=@n4, Alpha:float=1) -> (Pose:Component Space Pose Link Structure) {
    is=title FortAnimGraphNode_Flap
    targetComponent=RotationZ
    alphaBoolBlend {blendOption=Linear}
    sourceDrives [{
      bUseRange=true
      drivingCurve=null
      multiplier=1
      rangeMax=0.35
      rangeMin=-0.35
      remappedMax=0.35
      remappedMin=-0.35
      sourceComponent=RotationZ
      useQuaternion=true
      sourceBone {boneName=thigh_l}
     }]
    targetBone {boneName=dyn_rearLeg_l_03}
   }
   n6 = Flap Node - Bone: dyn_outerLeg_l_03!(ComponentPose:Component Space Pose Link Structure=@n5, Alpha:float=1) -> (Pose:Component Space Pose Link Structure) {
    is=title FortAnimGraphNode_Flap
    targetComponent=RotationZ
    alphaBoolBlend {blendOption=Linear}
    sourceDrives [{
      bUseRange=true
      drivingCurve=null
      multiplier=1
      rangeMax=0.35
      rangeMin=-0.35
      remappedMax=0.35
      remappedMin=-0.35
      sourceComponent=RotationZ
      useQuaternion=true
      sourceBone {boneName=thigh_l}
     }]
    targetBone {boneName=dyn_outerLeg_l_03}
   }
   n7 = Flap Node - Bone: dyn_frontLeg_l_03!(ComponentPose:Component Space Pose Link Structure=@n6, Alpha:float=1) -> (Pose:Component Space Pose Link Structure) {
    is=title FortAnimGraphNode_Flap
    targetComponent=RotationZ
    alphaBoolBlend {blendOption=Linear}
    sourceDrives [{
      bUseRange=true
      drivingCurve=null
      multiplier=1
      rangeMax=0.1725
      rangeMin=-0.1725
      remappedMax=0.1725
      remappedMin=-0.1725
      sourceComponent=RotationZ
      useQuaternion=true
      sourceBone {boneName=thigh_l}
     }]
    targetBone {boneName=dyn_frontLeg_l_03}
   }
   n8 = Get calf_rot_r() -> (calf_rot_r:double)
   n9 = "Transform (Modify) Bone - Bone: dyn_rearLeg_r_01"(ComponentPose=@n7, Alpha=@n8, Translation="-15,0,8", Rotation="-60,0,0", Scale="1,1,1") {
    rotationMode=BMM_Additive
    rotationSpace=BCS_BoneSpace
    translationMode=BMM_Additive
    translationSpace=BCS_ParentBoneSpace
    alphaBoolBlend {blendOption=Linear}
    alphaScaleBiasClamp {bInterpResult=true}
    boneToModify {boneName=dyn_rearLeg_r_01}
    rotation {pitch=-60}
    translation {x=-15, z=8}
   }
   n10 = "Transform (Modify) Bone - Bone: dyn_outerLeg_r_01"(ComponentPose=@n9, Alpha=@n8, Translation="-15,0,8", Rotation="-60,0,0", Scale="1,1,1") {
    rotationMode=BMM_Additive
    rotationSpace=BCS_BoneSpace
    translationMode=BMM_Additive
    translationSpace=BCS_ParentBoneSpace
    alphaBoolBlend {blendOption=Linear}
    alphaScaleBiasClamp {bInterpResult=true}
    boneToModify {boneName=dyn_outerLeg_r_01}
    rotation {pitch=-60}
    translation {x=-15, z=8}
   }
   n11 = Anim Dynamics - Bone: dyn_frontLeg_r_01(ComponentPose=@n10, Alpha=1, ExternalForce="0,0,0") {
    angularDampingOverride=1
    angularSpringConstant=1500
    bAngularSpring=true
    bGravityOverrideInSimSpace=true
    bOverrideAngularDamping=true
    bUseSphericalLimits=true
    gravityScale=0.35
    simulationSpace=RootRelative
    alphaBoolBlend {blendOption=Linear}
    alphaScaleBiasClamp {scale=0.5}
    boundBone {boneName=dyn_frontLeg_r_01}
    physicsBodyDefinitions [{
      collisionType=CustomSphere
      sphereCollisionRadius=2
      boundBone {boneName=dyn_frontLeg_r_01}
      boxExtents=<5 20 20>
      constraintSetup {angularLimitsMax=<5 20 10>, angularLimitsMin=<-5 -20 -40>, angularTarget=<0 0 0>}
      localJointOffset=<10 0 0>
     }]
    sphericalLimits [
     {limitRadius=16, drivingBone {boneName=pelvis}, sphereLocalOffset=<0 0 0>}
     {limitRadius=25, drivingBone {boneName=thigh_r}, sphereLocalOffset=<25 0 -20>}
     {limitRadius=20, drivingBone {boneName=calf_r}, sphereLocalOffset=<25 0 -20>}
     {limitRadius=14, drivingBone {boneName=thigh_r}, sphereLocalOffset=<25 0 0>}
     {limitRadius=14, drivingBone {boneName=thigh_r}, sphereLocalOffset=<35 0 0>}
     {limitRadius=14, drivingBone {boneName=calf_r}, sphereLocalOffset=<5 0 0>}
     {limitRadius=14, drivingBone {boneName=calf_r}, sphereLocalOffset=<15 0 0>}
     {limitRadius=14, drivingBone {boneName=calf_r}, sphereLocalOffset=<25 0 0>}
     {limitRadius=14, drivingBone {boneName=calf_r}, sphereLocalOffset=<35 0 0>}
    ]
   }
   n12 = Anim Dynamics - Bone: dyn_frontLeg_r_02(ComponentPose=@n11, Alpha=1, ExternalForce="0,0,0") {
    angularDampingOverride=1
    angularSpringConstant=1500
    bAngularSpring=true
    bGravityOverrideInSimSpace=true
    bOverrideAngularDamping=true
    bUseSphericalLimits=true
    gravityScale=0.35
    simulationSpace=RootRelative
    alphaBoolBlend {blendOption=Linear}
    alphaScaleBiasClamp {scale=0.5}
    boundBone {boneName=dyn_frontLeg_r_02}
    physicsBodyDefinitions [{
      collisionType=CustomSphere
      sphereCollisionRadius=2
      boundBone {boneName=dyn_frontLeg_r_02}
      boxExtents=<5 20 20>
      constraintSetup {angularLimitsMax=<5 20 10>, angularLimitsMin=<-5 -20 -40>, angularTarget=<0 0 0>}
      localJointOffset=<10 0 0>
     }]
    sphericalLimits [
     {limitRadius=16, drivingBone {boneName=pelvis}, sphereLocalOffset=<0 0 0>}
     {limitRadius=20, drivingBone {boneName=thigh_r}, sphereLocalOffset=<25 0 -20>}
     {limitRadius=20, drivingBone {boneName=calf_r}, sphereLocalOffset=<25 0 -20>}
     {limitRadius=14, drivingBone {boneName=thigh_r}, sphereLocalOffset=<25 0 0>}
     {limitRadius=14, drivingBone {boneName=thigh_r}, sphereLocalOffset=<35 0 0>}
     {limitRadius=14, drivingBone {boneName=calf_r}, sphereLocalOffset=<5 0 0>}
     {limitRadius=14, drivingBone {boneName=calf_r}, sphereLocalOffset=<15 0 0>}
     {limitRadius=14, drivingBone {boneName=calf_r}, sphereLocalOffset=<25 0 0>}
     {limitRadius=14, drivingBone {boneName=calf_r}, sphereLocalOffset=<35 0 0>}
    ]
   }
   n13 = Anim Dynamics - Bone: dyn_frontLeg_r_03(ComponentPose=@n12, Alpha=1, ExternalForce="0,0,0") {
    angularDampingOverride=1
    angularSpringConstant=1500
    bAngularSpring=true
    bGravityOverrideInSimSpace=true
    bOverrideAngularDamping=true
    bUseSphericalLimits=true
    gravityScale=0.35
    simulationSpace=RootRelative
    alphaBoolBlend {blendOption=Linear}
    alphaScaleBiasClamp {scale=0.5}
    boundBone {boneName=dyn_frontLeg_r_03}
    physicsBodyDefinitions [{
      collisionType=CustomSphere
      sphereCollisionRadius=2
      boundBone {boneName=dyn_frontLeg_r_03}
      boxExtents=<5 20 20>
      constraintSetup {angularLimitsMax=<5 20 10>, angularLimitsMin=<-5 -20 -40>, angularTarget=<0 0 0>}
      localJointOffset=<10 0 0>
     }]
    sphericalLimits [
     {limitRadius=16, drivingBone {boneName=pelvis}, sphereLocalOffset=<0 0 0>}
     {limitRadius=20, drivingBone {boneName=thigh_r}, sphereLocalOffset=<25 0 -20>}
     {limitRadius=20, drivingBone {boneName=calf_r}, sphereLocalOffset=<25 0 -20>}
     {limitRadius=14, drivingBone {boneName=thigh_r}, sphereLocalOffset=<25 0 0>}
     {limitRadius=14, drivingBone {boneName=thigh_r}, sphereLocalOffset=<35 0 0>}
     {limitRadius=14, drivingBone {boneName=calf_r}, sphereLocalOffset=<5 0 0>}
     {limitRadius=14, drivingBone {boneName=calf_r}, sphereLocalOffset=<15 0 0>}
     {limitRadius=14, drivingBone {boneName=calf_r}, sphereLocalOffset=<25 0 0>}
     {limitRadius=14, drivingBone {boneName=calf_r}, sphereLocalOffset=<35 0 0>}
    ]
   }
   n14 = Anim Dynamics - Bone: dyn_outerLeg_r_01(ComponentPose=@n13, Alpha=1, ExternalForce="0,0,0") {
    angularDampingOverride=1
    angularSpringConstant=1500
    bAngularSpring=true
    bGravityOverrideInSimSpace=true
    bOverrideAngularDamping=true
    bUseSphericalLimits=true
    gravityScale=0.35
    simulationSpace=RootRelative
    alphaBoolBlend {blendOption=Linear}
    alphaScaleBiasClamp {scale=0.5}
    boundBone {boneName=dyn_outerLeg_r_01}
    physicsBodyDefinitions [{
      collisionType=CustomSphere
      sphereCollisionRadius=2
      boundBone {boneName=dyn_outerLeg_r_01}
      boxExtents=<5 20 20>
      constraintSetup {angularLimitsMax=<5 20 10>, angularLimitsMin=<-5 -20 -40>, angularTarget=<0 0 0>}
      localJointOffset=<10 0 0>
     }]
    sphericalLimits [
     {limitRadius=16, drivingBone {boneName=pelvis}, sphereLocalOffset=<0 0 0>}
     {limitRadius=20, drivingBone {boneName=thigh_r}, sphereLocalOffset=<25 0 -20>}
     {limitRadius=20, drivingBone {boneName=calf_r}, sphereLocalOffset=<25 0 -20>}
     {limitRadius=14, drivingBone {boneName=thigh_r}, sphereLocalOffset=<25 0 0>}
     {limitRadius=14, drivingBone {boneName=thigh_r}, sphereLocalOffset=<35 0 0>}
     {limitRadius=14, drivingBone {boneName=calf_r}, sphereLocalOffset=<5 0 0>}
     {limitRadius=14, drivingBone {boneName=calf_r}, sphereLocalOffset=<15 0 0>}
     {limitRadius=14, drivingBone {boneName=calf_r}, sphereLocalOffset=<25 0 0>}
     {limitRadius=14, drivingBone {boneName=calf_r}, sphereLocalOffset=<35 0 0>}
    ]
   }
   n15 = Anim Dynamics - Bone: dyn_outerLeg_r_02(ComponentPose=@n14, Alpha=1, ExternalForce="0,0,0") {
    angularDampingOverride=1
    angularSpringConstant=1500
    bAngularSpring=true
    bGravityOverrideInSimSpace=true
    bOverrideAngularDamping=true
    bUseSphericalLimits=true
    gravityScale=0.35
    simulationSpace=RootRelative
    alphaBoolBlend {blendOption=Linear}
    alphaScaleBiasClamp {scale=0.5}
    boundBone {boneName=dyn_outerLeg_r_02}
    physicsBodyDefinitions [{
      collisionType=CustomSphere
      sphereCollisionRadius=2
      boundBone {boneName=dyn_outerLeg_r_02}
      boxExtents=<5 20 20>
      constraintSetup {angularLimitsMax=<5 20 10>, angularLimitsMin=<-5 -20 -40>, angularTarget=<0 0 0>}
      localJointOffset=<10 0 0>
     }]
    sphericalLimits [
     {limitRadius=16, drivingBone {boneName=pelvis}, sphereLocalOffset=<0 0 0>}
     {limitRadius=20, drivingBone {boneName=thigh_r}, sphereLocalOffset=<25 0 -20>}
     {limitRadius=20, drivingBone {boneName=calf_r}, sphereLocalOffset=<25 0 -20>}
     {limitRadius=14, drivingBone {boneName=thigh_r}, sphereLocalOffset=<25 0 0>}
     {limitRadius=14, drivingBone {boneName=thigh_r}, sphereLocalOffset=<35 0 0>}
     {limitRadius=14, drivingBone {boneName=calf_r}, sphereLocalOffset=<5 0 0>}
     {limitRadius=14, drivingBone {boneName=calf_r}, sphereLocalOffset=<15 0 0>}
     {limitRadius=14, drivingBone {boneName=calf_r}, sphereLocalOffset=<25 0 0>}
     {limitRadius=14, drivingBone {boneName=calf_r}, sphereLocalOffset=<35 0 0>}
    ]
   }
   n16 = Anim Dynamics - Bone: dyn_outerLeg_r_03(ComponentPose=@n15, Alpha=1, ExternalForce="0,0,0") {
    angularDampingOverride=1
    angularSpringConstant=1500
    bAngularSpring=true
    bGravityOverrideInSimSpace=true
    bOverrideAngularDamping=true
    bUseSphericalLimits=true
    gravityScale=0.35
    simulationSpace=RootRelative
    alphaBoolBlend {blendOption=Linear}
    alphaScaleBiasClamp {scale=0.5}
    boundBone {boneName=dyn_outerLeg_r_03}
    physicsBodyDefinitions [{
      collisionType=CustomSphere
      sphereCollisionRadius=2
      boundBone {boneName=dyn_outerLeg_r_03}
      boxExtents=<5 20 20>
      constraintSetup {angularLimitsMax=<5 20 10>, angularLimitsMin=<-5 -20 -40>, angularTarget=<0 0 0>}
      localJointOffset=<10 0 0>
     }]
    sphericalLimits [
     {limitRadius=16, drivingBone {boneName=pelvis}, sphereLocalOffset=<0 0 0>}
     {limitRadius=20, drivingBone {boneName=thigh_r}, sphereLocalOffset=<25 0 -20>}
     {limitRadius=20, drivingBone {boneName=calf_r}, sphereLocalOffset=<25 0 -20>}
     {limitRadius=14, drivingBone {boneName=thigh_r}, sphereLocalOffset=<25 0 0>}
     {limitRadius=14, drivingBone {boneName=thigh_r}, sphereLocalOffset=<35 0 0>}
     {limitRadius=14, drivingBone {boneName=calf_r}, sphereLocalOffset=<5 0 0>}
     {limitRadius=14, drivingBone {boneName=calf_r}, sphereLocalOffset=<15 0 0>}
     {limitRadius=14, drivingBone {boneName=calf_r}, sphereLocalOffset=<25 0 0>}
     {limitRadius=14, drivingBone {boneName=calf_r}, sphereLocalOffset=<35 0 0>}
    ]
   }
   n17 = Anim Dynamics - Bone: dyn_rearLeg_r_01(ComponentPose=@n16, Alpha=1, ExternalForce="0,0,0") {
    angularDampingOverride=1
    angularSpringConstant=1500
    bAngularSpring=true
    bGravityOverrideInSimSpace=true
    bOverrideAngularDamping=true
    bUseSphericalLimits=true
    gravityScale=0.35
    simulationSpace=RootRelative
    alphaBoolBlend {blendOption=Linear}
    alphaScaleBiasClamp {scale=0.5}
    boundBone {boneName=dyn_rearLeg_r_01}
    physicsBodyDefinitions [{
      collisionType=CustomSphere
      sphereCollisionRadius=2
      boundBone {boneName=dyn_rearLeg_r_01}
      boxExtents=<5 20 20>
      constraintSetup {angularLimitsMax=<5 20 10>, angularLimitsMin=<-5 -20 -40>, angularTarget=<0 0 0>}
      localJointOffset=<10 0 0>
     }]
    sphericalLimits [
     {limitRadius=16, drivingBone {boneName=pelvis}, sphereLocalOffset=<0 0 0>}
     {limitRadius=20, drivingBone {boneName=thigh_r}, sphereLocalOffset=<25 0 -20>}
     {limitRadius=20, drivingBone {boneName=calf_r}, sphereLocalOffset=<25 0 -20>}
     {limitRadius=14, drivingBone {boneName=thigh_r}, sphereLocalOffset=<25 0 0>}
     {limitRadius=14, drivingBone {boneName=thigh_r}, sphereLocalOffset=<35 0 0>}
     {limitRadius=14, drivingBone {boneName=calf_r}, sphereLocalOffset=<5 0 0>}
     {limitRadius=14, drivingBone {boneName=calf_r}, sphereLocalOffset=<15 0 0>}
     {limitRadius=14, drivingBone {boneName=calf_r}, sphereLocalOffset=<25 0 0>}
     {limitRadius=14, drivingBone {boneName=calf_r}, sphereLocalOffset=<35 0 0>}
    ]
   }
   n18 = Anim Dynamics - Bone: dyn_rearLeg_r_02(ComponentPose=@n17, Alpha=1, ExternalForce="0,0,0") {
    angularDampingOverride=1
    angularSpringConstant=1500
    bAngularSpring=true
    bGravityOverrideInSimSpace=true
    bOverrideAngularDamping=true
    bUseSphericalLimits=true
    gravityScale=0.35
    simulationSpace=RootRelative
    alphaBoolBlend {blendOption=Linear}
    alphaScaleBiasClamp {scale=0.5}
    boundBone {boneName=dyn_rearLeg_r_02}
    physicsBodyDefinitions [{
      collisionType=CustomSphere
      sphereCollisionRadius=2
      boundBone {boneName=dyn_rearLeg_r_02}
      boxExtents=<5 20 20>
      constraintSetup {angularLimitsMax=<5 20 10>, angularLimitsMin=<-5 -20 -40>, angularTarget=<0 0 0>}
      localJointOffset=<10 0 0>
     }]
    sphericalLimits [
     {limitRadius=16, drivingBone {boneName=pelvis}, sphereLocalOffset=<0 0 0>}
     {limitRadius=20, drivingBone {boneName=thigh_r}, sphereLocalOffset=<25 0 -20>}
     {limitRadius=20, drivingBone {boneName=calf_r}, sphereLocalOffset=<25 0 -20>}
     {limitRadius=14, drivingBone {boneName=thigh_r}, sphereLocalOffset=<25 0 0>}
     {limitRadius=14, drivingBone {boneName=thigh_r}, sphereLocalOffset=<35 0 0>}
     {limitRadius=14, drivingBone {boneName=calf_r}, sphereLocalOffset=<5 0 0>}
     {limitRadius=14, drivingBone {boneName=calf_r}, sphereLocalOffset=<15 0 0>}
     {limitRadius=14, drivingBone {boneName=calf_r}, sphereLocalOffset=<25 0 0>}
     {limitRadius=14, drivingBone {boneName=calf_r}, sphereLocalOffset=<35 0 0>}
    ]
   }
   n19 = Anim Dynamics - Bone: dyn_rearLeg_r_03(ComponentPose=@n18, Alpha=1, ExternalForce="0,0,0") {
    angularDampingOverride=1
    angularSpringConstant=1500
    bAngularSpring=true
    bGravityOverrideInSimSpace=true
    bOverrideAngularDamping=true
    bUseSphericalLimits=true
    gravityScale=0.35
    simulationSpace=RootRelative
    alphaBoolBlend {blendOption=Linear}
    alphaScaleBiasClamp {scale=0.5}
    boundBone {boneName=dyn_rearLeg_r_03}
    physicsBodyDefinitions [{
      collisionType=CustomSphere
      sphereCollisionRadius=2
      boundBone {boneName=dyn_rearLeg_r_03}
      boxExtents=<5 20 20>
      constraintSetup {angularLimitsMax=<5 20 10>, angularLimitsMin=<-5 -20 -40>, angularTarget=<0 0 0>}
      localJointOffset=<10 0 0>
     }]
    sphericalLimits [
     {limitRadius=16, drivingBone {boneName=pelvis}, sphereLocalOffset=<0 0 0>}
     {limitRadius=20, drivingBone {boneName=thigh_r}, sphereLocalOffset=<25 0 -20>}
     {limitRadius=20, drivingBone {boneName=calf_r}, sphereLocalOffset=<25 0 -20>}
     {limitRadius=14, drivingBone {boneName=thigh_r}, sphereLocalOffset=<25 0 0>}
     {limitRadius=14, drivingBone {boneName=thigh_r}, sphereLocalOffset=<35 0 0>}
     {limitRadius=14, drivingBone {boneName=calf_r}, sphereLocalOffset=<5 0 0>}
     {limitRadius=14, drivingBone {boneName=calf_r}, sphereLocalOffset=<15 0 0>}
     {limitRadius=14, drivingBone {boneName=calf_r}, sphereLocalOffset=<25 0 0>}
     {limitRadius=14, drivingBone {boneName=calf_r}, sphereLocalOffset=<35 0 0>}
    ]
   }
   n20 = Anim Dynamics - Bone: dyn_frontLeg_l_01(ComponentPose=@n19, Alpha=1, ExternalForce="0,0,0") {
    angularDampingOverride=1
    angularSpringConstant=1500
    bAngularSpring=true
    bGravityOverrideInSimSpace=true
    bOverrideAngularDamping=true
    bUseSphericalLimits=true
    gravityScale=0.35
    simulationSpace=RootRelative
    alphaBoolBlend {blendOption=Linear}
    alphaScaleBiasClamp {scale=0.5}
    boundBone {boneName=dyn_frontLeg_l_01}
    physicsBodyDefinitions [{
      collisionType=CustomSphere
      sphereCollisionRadius=2
      boundBone {boneName=dyn_frontLeg_l_01}
      boxExtents=<5 20 20>
      constraintSetup {angularLimitsMax=<5 40 40>, angularLimitsMin=<-5 -20 -40>, angularTarget=<0 0 0>}
      localJointOffset=<10 0 0>
     }]
    sphericalLimits [
     {limitRadius=16, drivingBone {boneName=pelvis}, sphereLocalOffset=<0 0 0>}
     {limitRadius=20, drivingBone {boneName=thigh_l}, sphereLocalOffset=<-25 0 20>}
     {limitRadius=20, drivingBone {boneName=calf_l}, sphereLocalOffset=<-25 0 20>}
     {limitRadius=16, drivingBone {boneName=thigh_l}, sphereLocalOffset=<-25 0 0>}
     {limitRadius=16, drivingBone {boneName=thigh_l}, sphereLocalOffset=<-35 0 0>}
     {limitRadius=14, drivingBone {boneName=calf_l}, sphereLocalOffset=<-5 0 0>}
     {limitRadius=14, drivingBone {boneName=calf_l}, sphereLocalOffset=<-15 0 0>}
     {limitRadius=14, drivingBone {boneName=calf_l}, sphereLocalOffset=<-25 0 0>}
     {limitRadius=14, drivingBone {boneName=calf_l}, sphereLocalOffset=<-35 0 0>}
    ]
   }
   n21 = Anim Dynamics - Bone: dyn_frontLeg_l_02(ComponentPose=@n20, Alpha=1, ExternalForce="0,0,0") {
    angularDampingOverride=1
    angularSpringConstant=1500
    bAngularSpring=true
    bGravityOverrideInSimSpace=true
    bOverrideAngularDamping=true
    bUseSphericalLimits=true
    gravityScale=0.35
    simulationSpace=RootRelative
    alphaBoolBlend {blendOption=Linear}
    alphaScaleBiasClamp {scale=0.5}
    boundBone {boneName=dyn_frontLeg_l_02}
    physicsBodyDefinitions [{
      collisionType=CustomSphere
      sphereCollisionRadius=2
      boundBone {boneName=dyn_frontLeg_l_02}
      boxExtents=<5 20 20>
      constraintSetup {angularLimitsMax=<5 40 40>, angularLimitsMin=<-5 -20 -40>, angularTarget=<0 0 0>}
      localJointOffset=<10 0 0>
     }]
    sphericalLimits [
     {limitRadius=16, drivingBone {boneName=pelvis}, sphereLocalOffset=<0 0 0>}
     {limitRadius=20, drivingBone {boneName=thigh_l}, sphereLocalOffset=<-25 0 20>}
     {limitRadius=20, drivingBone {boneName=calf_l}, sphereLocalOffset=<-25 0 20>}
     {limitRadius=16, drivingBone {boneName=thigh_l}, sphereLocalOffset=<-25 0 0>}
     {limitRadius=16, drivingBone {boneName=thigh_l}, sphereLocalOffset=<-35 0 0>}
     {limitRadius=14, drivingBone {boneName=calf_l}, sphereLocalOffset=<-5 0 0>}
     {limitRadius=14, drivingBone {boneName=calf_l}, sphereLocalOffset=<-15 0 0>}
     {limitRadius=14, drivingBone {boneName=calf_l}, sphereLocalOffset=<-25 0 0>}
     {limitRadius=14, drivingBone {boneName=calf_l}, sphereLocalOffset=<-35 0 0>}
    ]
   }
   n22 = Anim Dynamics - Bone: dyn_frontLeg_l_03(ComponentPose=@n21, Alpha=1, ExternalForce="0,0,0") {
    angularDampingOverride=1
    angularSpringConstant=1500
    bAngularSpring=true
    bGravityOverrideInSimSpace=true
    bOverrideAngularDamping=true
    bUseSphericalLimits=true
    gravityScale=0.35
    simulationSpace=RootRelative
    alphaBoolBlend {blendOption=Linear}
    alphaScaleBiasClamp {scale=0.5}
    boundBone {boneName=dyn_frontLeg_l_03}
    physicsBodyDefinitions [{
      collisionType=CustomSphere
      sphereCollisionRadius=2
      boundBone {boneName=dyn_frontLeg_l_03}
      boxExtents=<5 20 20>
      constraintSetup {angularLimitsMax=<5 40 40>, angularLimitsMin=<-5 -20 -40>, angularTarget=<0 0 0>}
      localJointOffset=<10 0 0>
     }]
    sphericalLimits [
     {limitRadius=16, drivingBone {boneName=pelvis}, sphereLocalOffset=<0 0 0>}
     {limitRadius=20, drivingBone {boneName=thigh_l}, sphereLocalOffset=<-25 0 20>}
     {limitRadius=20, drivingBone {boneName=calf_l}, sphereLocalOffset=<-25 0 20>}
     {limitRadius=16, drivingBone {boneName=thigh_l}, sphereLocalOffset=<-25 0 0>}
     {limitRadius=16, drivingBone {boneName=thigh_l}, sphereLocalOffset=<-35 0 0>}
     {limitRadius=14, drivingBone {boneName=calf_l}, sphereLocalOffset=<-5 0 0>}
     {limitRadius=14, drivingBone {boneName=calf_l}, sphereLocalOffset=<-15 0 0>}
     {limitRadius=14, drivingBone {boneName=calf_l}, sphereLocalOffset=<-25 0 0>}
     {limitRadius=14, drivingBone {boneName=calf_l}, sphereLocalOffset=<-35 0 0>}
    ]
   }
   n23 = Anim Dynamics - Bone: dyn_outerLeg_l_01(ComponentPose=@n22, Alpha=1, ExternalForce="0,0,0") {
    angularDampingOverride=1
    angularSpringConstant=1500
    bAngularSpring=true
    bGravityOverrideInSimSpace=true
    bOverrideAngularDamping=true
    bUseSphericalLimits=true
    gravityScale=0.35
    simulationSpace=RootRelative
    alphaBoolBlend {blendOption=Linear}
    alphaScaleBiasClamp {scale=0.5}
    boundBone {boneName=dyn_outerLeg_l_01}
    physicsBodyDefinitions [{
      collisionType=CustomSphere
      sphereCollisionRadius=2
      boundBone {boneName=dyn_outerLeg_l_01}
      boxExtents=<5 20 20>
      constraintSetup {angularLimitsMax=<5 40 40>, angularLimitsMin=<-5 -40 -40>, angularTarget=<0 0 0>}
      localJointOffset=<10 0 0>
     }]
    sphericalLimits [
     {limitRadius=16, drivingBone {boneName=pelvis}, sphereLocalOffset=<0 0 0>}
     {limitRadius=20, drivingBone {boneName=thigh_l}, sphereLocalOffset=<-25 0 20>}
     {limitRadius=20, drivingBone {boneName=calf_l}, sphereLocalOffset=<-25 0 20>}
     {limitRadius=16, drivingBone {boneName=thigh_l}, sphereLocalOffset=<-25 0 0>}
     {limitRadius=16, drivingBone {boneName=thigh_l}, sphereLocalOffset=<-35 0 0>}
     {limitRadius=14, drivingBone {boneName=calf_l}, sphereLocalOffset=<-5 0 0>}
     {limitRadius=14, drivingBone {boneName=calf_l}, sphereLocalOffset=<-15 0 0>}
     {limitRadius=14, drivingBone {boneName=calf_l}, sphereLocalOffset=<-25 0 0>}
     {limitRadius=14, drivingBone {boneName=calf_l}, sphereLocalOffset=<-35 0 0>}
    ]
   }
   n24 = Anim Dynamics - Bone: dyn_outerLeg_l_02(ComponentPose=@n23, Alpha=1, ExternalForce="0,0,0") {
    angularDampingOverride=1
    angularSpringConstant=1500
    bAngularSpring=true
    bGravityOverrideInSimSpace=true
    bOverrideAngularDamping=true
    bUseSphericalLimits=true
    gravityScale=0.35
    simulationSpace=RootRelative
    alphaBoolBlend {blendOption=Linear}
    alphaScaleBiasClamp {scale=0.5}
    boundBone {boneName=dyn_outerLeg_l_02}
    physicsBodyDefinitions [{
      collisionType=CustomSphere
      sphereCollisionRadius=2
      boundBone {boneName=dyn_outerLeg_l_02}
      boxExtents=<5 20 20>
      constraintSetup {angularLimitsMax=<5 40 40>, angularLimitsMin=<-5 -40 -40>, angularTarget=<0 0 0>}
      localJointOffset=<10 0 0>
     }]
    sphericalLimits [
     {limitRadius=16, drivingBone {boneName=pelvis}, sphereLocalOffset=<0 0 0>}
     {limitRadius=20, drivingBone {boneName=thigh_l}, sphereLocalOffset=<-25 0 20>}
     {limitRadius=20, drivingBone {boneName=calf_l}, sphereLocalOffset=<-25 0 20>}
     {limitRadius=16, drivingBone {boneName=thigh_l}, sphereLocalOffset=<-25 0 0>}
     {limitRadius=16, drivingBone {boneName=thigh_l}, sphereLocalOffset=<-35 0 0>}
     {limitRadius=14, drivingBone {boneName=calf_l}, sphereLocalOffset=<-5 0 0>}
     {limitRadius=14, drivingBone {boneName=calf_l}, sphereLocalOffset=<-15 0 0>}
     {limitRadius=14, drivingBone {boneName=calf_l}, sphereLocalOffset=<-25 0 0>}
     {limitRadius=14, drivingBone {boneName=calf_l}, sphereLocalOffset=<-35 0 0>}
    ]
   }
   n25 = Anim Dynamics - Bone: dyn_outerLeg_l_03(ComponentPose=@n24, Alpha=1, ExternalForce="0,0,0") {
    angularDampingOverride=1
    angularSpringConstant=1500
    bAngularSpring=true
    bGravityOverrideInSimSpace=true
    bOverrideAngularDamping=true
    bUseSphericalLimits=true
    gravityScale=0.35
    simulationSpace=RootRelative
    alphaBoolBlend {blendOption=Linear}
    alphaScaleBiasClamp {scale=0.5}
    boundBone {boneName=dyn_outerLeg_l_03}
    physicsBodyDefinitions [{
      collisionType=CustomSphere
      sphereCollisionRadius=2
      boundBone {boneName=dyn_outerLeg_l_03}
      boxExtents=<5 20 20>
      constraintSetup {angularLimitsMax=<5 40 40>, angularLimitsMin=<-5 -40 -40>, angularTarget=<0 0 0>}
      localJointOffset=<10 0 0>
     }]
    sphericalLimits [
     {limitRadius=16, drivingBone {boneName=pelvis}, sphereLocalOffset=<0 0 0>}
     {limitRadius=20, drivingBone {boneName=thigh_l}, sphereLocalOffset=<-25 0 20>}
     {limitRadius=20, drivingBone {boneName=calf_l}, sphereLocalOffset=<-25 0 20>}
     {limitRadius=16, drivingBone {boneName=thigh_l}, sphereLocalOffset=<-25 0 0>}
     {limitRadius=16, drivingBone {boneName=thigh_l}, sphereLocalOffset=<-35 0 0>}
     {limitRadius=14, drivingBone {boneName=calf_l}, sphereLocalOffset=<-5 0 0>}
     {limitRadius=14, drivingBone {boneName=calf_l}, sphereLocalOffset=<-15 0 0>}
     {limitRadius=14, drivingBone {boneName=calf_l}, sphereLocalOffset=<-25 0 0>}
     {limitRadius=14, drivingBone {boneName=calf_l}, sphereLocalOffset=<-35 0 0>}
    ]
   }
   n26 = Anim Dynamics - Bone: dyn_rearLeg_l_01(ComponentPose=@n25, Alpha=1, ExternalForce="0,0,0") {
    angularDampingOverride=1
    angularSpringConstant=1500
    bAngularSpring=true
    bGravityOverrideInSimSpace=true
    bOverrideAngularDamping=true
    bUseSphericalLimits=true
    gravityScale=0.35
    simulationSpace=RootRelative
    alphaBoolBlend {blendOption=Linear}
    alphaScaleBiasClamp {scale=0.5}
    boundBone {boneName=dyn_rearLeg_l_01}
    physicsBodyDefinitions [{
      collisionType=CustomSphere
      sphereCollisionRadius=2
      boundBone {boneName=dyn_rearLeg_l_01}
      boxExtents=<5 20 20>
      constraintSetup {angularLimitsMax=<5 40 40>, angularLimitsMin=<-5 -40 -40>, angularTarget=<0 0 0>}
      localJointOffset=<10 0 0>
     }]
    sphericalLimits [
     {limitRadius=16, drivingBone {boneName=pelvis}, sphereLocalOffset=<0 0 0>}
     {limitRadius=20, drivingBone {boneName=thigh_l}, sphereLocalOffset=<-25 0 20>}
     {limitRadius=20, drivingBone {boneName=calf_l}, sphereLocalOffset=<-25 0 20>}
     {limitRadius=16, drivingBone {boneName=thigh_l}, sphereLocalOffset=<-25 0 0>}
     {limitRadius=16, drivingBone {boneName=thigh_l}, sphereLocalOffset=<-35 0 0>}
     {limitRadius=14, drivingBone {boneName=calf_l}, sphereLocalOffset=<-5 0 0>}
     {limitRadius=14, drivingBone {boneName=calf_l}, sphereLocalOffset=<-15 0 0>}
     {limitRadius=14, drivingBone {boneName=calf_l}, sphereLocalOffset=<-25 0 0>}
     {limitRadius=14, drivingBone {boneName=calf_l}, sphereLocalOffset=<-35 0 0>}
    ]
   }
   n27 = Anim Dynamics - Bone: dyn_rearLeg_l_02(ComponentPose=@n26, Alpha=1, ExternalForce="0,0,0") {
    angularDampingOverride=1
    angularSpringConstant=1500
    bAngularSpring=true
    bGravityOverrideInSimSpace=true
    bOverrideAngularDamping=true
    bUseSphericalLimits=true
    gravityScale=0.35
    simulationSpace=RootRelative
    alphaBoolBlend {blendOption=Linear}
    alphaScaleBiasClamp {scale=0.5}
    boundBone {boneName=dyn_rearLeg_l_02}
    physicsBodyDefinitions [{
      collisionType=CustomSphere
      sphereCollisionRadius=2
      boundBone {boneName=dyn_rearLeg_l_02}
      boxExtents=<5 20 20>
      constraintSetup {angularLimitsMax=<5 40 40>, angularLimitsMin=<-5 -40 -40>, angularTarget=<0 0 0>}
      localJointOffset=<10 0 0>
     }]
    sphericalLimits [
     {limitRadius=16, drivingBone {boneName=pelvis}, sphereLocalOffset=<0 0 0>}
     {limitRadius=20, drivingBone {boneName=thigh_l}, sphereLocalOffset=<-25 0 20>}
     {limitRadius=20, drivingBone {boneName=calf_l}, sphereLocalOffset=<-25 0 20>}
     {limitRadius=16, drivingBone {boneName=thigh_l}, sphereLocalOffset=<-25 0 0>}
     {limitRadius=16, drivingBone {boneName=thigh_l}, sphereLocalOffset=<-35 0 0>}
     {limitRadius=14, drivingBone {boneName=calf_l}, sphereLocalOffset=<-5 0 0>}
     {limitRadius=14, drivingBone {boneName=calf_l}, sphereLocalOffset=<-15 0 0>}
     {limitRadius=14, drivingBone {boneName=calf_l}, sphereLocalOffset=<-25 0 0>}
     {limitRadius=14, drivingBone {boneName=calf_l}, sphereLocalOffset=<-35 0 0>}
    ]
   }
   n28 = Anim Dynamics - Bone: dyn_rearLeg_l_03(ComponentPose=@n27, Alpha=1, ExternalForce="0,0,0") {
    angularDampingOverride=1
    angularSpringConstant=1500
    bAngularSpring=true
    bGravityOverrideInSimSpace=true
    bOverrideAngularDamping=true
    bUseSphericalLimits=true
    gravityScale=0.35
    simulationSpace=RootRelative
    alphaBoolBlend {blendOption=Linear}
    alphaScaleBiasClamp {scale=0.5}
    boundBone {boneName=dyn_rearLeg_l_03}
    physicsBodyDefinitions [{
      collisionType=CustomSphere
      sphereCollisionRadius=2
      boundBone {boneName=dyn_rearLeg_l_03}
      boxExtents=<5 20 20>
      constraintSetup {angularLimitsMax=<5 40 40>, angularLimitsMin=<-5 -40 -40>, angularTarget=<0 0 0>}
      localJointOffset=<10 0 0>
     }]
    sphericalLimits [
     {limitRadius=16, drivingBone {boneName=pelvis}, sphereLocalOffset=<0 0 0>}
     {limitRadius=20, drivingBone {boneName=thigh_l}, sphereLocalOffset=<-25 0 20>}
     {limitRadius=20, drivingBone {boneName=calf_l}, sphereLocalOffset=<-25 0 20>}
     {limitRadius=16, drivingBone {boneName=thigh_l}, sphereLocalOffset=<-25 0 0>}
     {limitRadius=16, drivingBone {boneName=thigh_l}, sphereLocalOffset=<-35 0 0>}
     {limitRadius=14, drivingBone {boneName=calf_l}, sphereLocalOffset=<-5 0 0>}
     {limitRadius=14, drivingBone {boneName=calf_l}, sphereLocalOffset=<-15 0 0>}
     {limitRadius=14, drivingBone {boneName=calf_l}, sphereLocalOffset=<-25 0 0>}
     {limitRadius=14, drivingBone {boneName=calf_l}, sphereLocalOffset=<-35 0 0>}
    ]
   }
   n29 = "Transform (Modify) Bone - Bone: dyn_l_mstr"(ComponentPose=@n28, Alpha=1, Translation="0,0,5", Rotation="20,0,0", Scale="1,1,1") {
    rotationMode=BMM_Additive
    rotationSpace=BCS_BoneSpace
    scaleMode=BMM_Additive
    scaleSpace=BCS_BoneSpace
    translationMode=BMM_Additive
    translationSpace=BCS_BoneSpace
    alphaBoolBlend {blendOption=Linear}
    alphaScaleBiasClamp {bInterpResult=true}
    boneToModify {boneName=dyn_l_mstr}
    rotation {pitch=20}
    translation {z=5}
   }
   n30 = "Transform (Modify) Bone - Bone: dyn_rearLeg_r_01"(ComponentPose=@n29, Alpha=1, Translation="0,0,-5", Rotation="-20,0,0", Scale="1,1,1") {
    rotationMode=BMM_Additive
    rotationSpace=BCS_BoneSpace
    scaleMode=BMM_Additive
    scaleSpace=BCS_BoneSpace
    translationMode=BMM_Additive
    translationSpace=BCS_BoneSpace
    alphaBoolBlend {blendOption=Linear}
    alphaScaleBiasClamp {bInterpResult=true}
    boneToModify {boneName=dyn_rearLeg_r_01}
    rotation {pitch=-20}
    translation {z=-5}
   }
   n31 = Flap Node - Bone: dyn_rearLeg_r_03(ComponentPose=@n30, Alpha=1) {
    targetComponent=ScaleX
    alphaBoolBlend {blendOption=Linear}
    sourceDrives [{
      bUseRange=true
      drivingCurve=null
      multiplier=1
      rangeMax=0
      rangeMin=-3
      remappedMax=0
      remappedMin=-1.7
      sourceComponent=RotationZ
      useQuaternion=true
      sourceBone {boneName=calf_r}
     }]
    targetBone {boneName=dyn_rearLeg_r_03}
   }
   n32 = Flap Node - Bone: dyn_outerLeg_r_03(ComponentPose=@n31, Alpha=1) {
    targetComponent=ScaleX
    alphaBoolBlend {blendOption=Linear}
    sourceDrives [{
      bUseRange=true
      drivingCurve=null
      multiplier=1
      rangeMax=0
      rangeMin=-2.094
      remappedMax=0
      remappedMin=-0.9
      sourceComponent=RotationZ
      useQuaternion=true
      sourceBone {boneName=calf_r}
     }]
    targetBone {boneName=dyn_outerLeg_r_03}
   }
   n33 = Flap Node - Bone: dyn_frontLeg_r_03(ComponentPose=@n32, Alpha=1) {
    targetComponent=ScaleX
    alphaBoolBlend {blendOption=Linear}
    sourceDrives [{
      bUseRange=true
      drivingCurve=null
      multiplier=1
      rangeMax=0
      rangeMin=-1.2
      remappedMax=0
      remappedMin=-0.6
      sourceComponent=RotationZ
      useQuaternion=true
      sourceBone {boneName=calf_r}
     }]
    targetBone {boneName=dyn_frontLeg_r_03}
   }
   n34 = Flap Node - Bone: dyn_rearLeg_l_03(ComponentPose=@n33, Alpha=1) {
    targetComponent=ScaleX
    alphaBoolBlend {blendOption=Linear}
    sourceDrives [{
      bUseRange=true
      drivingCurve=null
      multiplier=1
      rangeMax=0
      rangeMin=-3
      remappedMax=0
      remappedMin=-1.7
      sourceComponent=RotationZ
      useQuaternion=true
      sourceBone {boneName=calf_l}
     }]
    targetBone {boneName=dyn_rearLeg_l_03}
   }
   n35 = Flap Node - Bone: dyn_outerLeg_l_03(ComponentPose=@n34, Alpha=1) {
    targetComponent=ScaleX
    alphaBoolBlend {blendOption=Linear}
    sourceDrives [{
      bUseRange=true
      drivingCurve=null
      multiplier=1
      rangeMax=0
      rangeMin=-2.094
      remappedMax=0
      remappedMin=-0.9
      sourceComponent=RotationZ
      useQuaternion=true
      sourceBone {boneName=calf_l}
     }]
    targetBone {boneName=dyn_outerLeg_l_03}
   }
   n36 = Flap Node - Bone: dyn_frontLeg_l_03(ComponentPose=@n35, Alpha=1) {
    targetComponent=ScaleX
    alphaBoolBlend {blendOption=Linear}
    sourceDrives [{
      bUseRange=true
      drivingCurve=null
      multiplier=1
      rangeMax=0
      rangeMin=-1.2
      remappedMax=0
      remappedMin=-0.6
      sourceComponent=RotationZ
      useQuaternion=true
      sourceBone {boneName=calf_l}
     }]
    targetBone {boneName=dyn_frontLeg_l_03}
   }
   n37 = "Constraint - dyn_lightbox (PPPPPP)"!(
    ComponentPose=@n36
    Alpha=1
    ConstraintWeights_0=0.1324
    ConstraintWeights_1:float=0.0729
    ConstraintWeights_2:float=0.0601
    ConstraintWeights_3:float=0.0001
    ConstraintWeights_4:float=0.0002
    ConstraintWeights_5:float=0.7344
   ) -> (Pose) {
    alphaBoolBlend {blendOption=Linear}
    boneToModify {boneName=dyn_lightbox}
    constraintSetup [
     {targetBone {boneName=spine_01}}
     {targetBone {boneName=spine_02}}
     {targetBone {boneName=spine_03}}
     {targetBone {boneName=spine_04}}
     {targetBone {boneName=clavicle_r}}
     {targetBone {boneName=pelvis}}
    ]
    constraintWeights [1,1,1,1,1,1]
   }
   n38 = Anim Dynamics - Bone: dyn_lightbox(ComponentPose=@n37, Alpha=1, ExternalForce="0,0,0") {
    angularDampingOverride=0.9999
    bGravityOverrideInSimSpace=true
    bOverrideAngularDamping=true
    bUseSphericalLimits=true
    simulationSpace=RootRelative
    alphaBoolBlend {blendOption=Linear}
    boundBone {boneName=dyn_lightbox}
    physicsBodyDefinitions [{
      sphereCollisionRadius=2
      boundBone {boneName=dyn_lightbox}
      boxExtents=<10 20 20>
      constraintSetup {angularLimitsMax=<0 20 0>, angularLimitsMin=<0 -20 0>, angularTarget=<0 0 0>}
      localJointOffset=<10 0 0>
     }]
    sphericalLimits [{limitRadius=9, drivingBone {boneName=thigh_r}, sphereLocalOffset=<0 0 0>}]
   }
   n39 = "Transform (Modify) Bone - Bone: dyn_collar"!(ComponentPose=@n38, bAlphaBoolEnabled:bool=True, Translation="-3,-3,0", Rotation="0,-40,0", Scale="1,1.1,1.1") -> (Pose) {
    alphaInputType=Bool
    rotationMode=BMM_Additive
    rotationSpace=BCS_BoneSpace
    scaleMode=BMM_Additive
    scaleSpace=BCS_BoneSpace
    translationMode=BMM_Additive
    translationSpace=BCS_BoneSpace
    alphaBoolBlend {blendInTime=0.1, blendOption=Linear, blendOutTime=0.1}
    alphaScaleBiasClamp {bInterpResult=true}
    boneToModify {boneName=dyn_collar}
    rotation {yaw=-40}
    scale {y=1.1, z=1.1}
    translation=<-3 -3>
   }
   n40 = "Transform (Modify) Bone - Bone: dyn_rearLeg_l_01"!(ComponentPose=@n39, bAlphaBoolEnabled:bool=True, Translation="-0.667,-2.365,0.492", Rotation="18.152,-10.569,4.457", Scale="1,1,1") -> (Pose) {
    alphaInputType=Bool
    rotationMode=BMM_Additive
    rotationSpace=BCS_BoneSpace
    scaleMode=BMM_Additive
    scaleSpace=BCS_BoneSpace
    translationMode=BMM_Additive
    translationSpace=BCS_BoneSpace
    alphaBoolBlend {blendInTime=0.1, blendOption=Linear, blendOutTime=0.1}
    alphaScaleBiasClamp {bInterpResult=true}
    boneToModify {boneName=dyn_rearLeg_l_01}
    rotation=<rot 18.152 -10.5689 4.457>
    translation=<-0.667 -2.365 0.492>
   }
   n41 = "Transform (Modify) Bone - Bone: dyn_outerLeg_l_01"!(ComponentPose=@n40, bAlphaBoolEnabled:bool=True, Translation="-0.667,-2.365,0.492", Rotation="18.500,-21.064,1.178", Scale="1,1,1") -> (Pose) {
    alphaInputType=Bool
    rotationMode=BMM_Additive
    rotationSpace=BCS_BoneSpace
    scaleMode=BMM_Additive
    scaleSpace=BCS_BoneSpace
    translationMode=BMM_Additive
    translationSpace=BCS_BoneSpace
    alphaBoolBlend {blendInTime=0.1, blendOption=Linear, blendOutTime=0.1}
    alphaScaleBiasClamp {bInterpResult=true}
    boneToModify {boneName=dyn_outerLeg_l_01}
    rotation=<rot 18.5 -21.064 1.178>
    translation=<-0.667 -2.365 0.492>
   }
   n42 = "Transform (Modify) Bone - Bone: dyn_rearLeg_l_01"(ComponentPose=@n41, bAlphaBoolEnabled=True, Translation="-0.667,-2.365,0.492", Rotation="18.152,-10.569,4.457", Scale="1,1,1") {
    alphaInputType=Bool
    rotationMode=BMM_Additive
    rotationSpace=BCS_BoneSpace
    scaleMode=BMM_Additive
    scaleSpace=BCS_BoneSpace
    translationMode=BMM_Additive
    translationSpace=BCS_BoneSpace
    alphaBoolBlend {blendInTime=0.1, blendOption=Linear, blendOutTime=0.1}
    alphaScaleBiasClamp {bInterpResult=true}
    boneToModify {boneName=dyn_rearLeg_l_01}
    rotation=<rot 18.152 -10.5689 4.457>
    translation=<-0.667 -2.365 0.492>
   }
   n43 = "Transform (Modify) Bone - Bone: dyn_outerLeg_l_01"(ComponentPose=@n42, bAlphaBoolEnabled=True, Translation="-0.667,-2.365,0.492", Rotation="18.500,-21.064,1.178", Scale="1,1,1") {
    alphaInputType=Bool
    rotationMode=BMM_Additive
    rotationSpace=BCS_BoneSpace
    scaleMode=BMM_Additive
    scaleSpace=BCS_BoneSpace
    translationMode=BMM_Additive
    translationSpace=BCS_BoneSpace
    alphaBoolBlend {blendInTime=0.1, blendOption=Linear, blendOutTime=0.1}
    alphaScaleBiasClamp {bInterpResult=true}
    boneToModify {boneName=dyn_outerLeg_l_01}
    rotation=<rot 18.5 -21.064 1.178>
    translation=<-0.667 -2.365 0.492>
   }
   n44 = "Transform (Modify) Bone - Bone: dyn_rearLeg_l_01"(ComponentPose=@n43, bAlphaBoolEnabled=True, Translation="-6.481,-6.076,2.307", Rotation="27.717,-30.695,7.779", Scale="1,1,1") {
    alphaInputType=Bool
    rotationMode=BMM_Additive
    rotationSpace=BCS_BoneSpace
    scaleMode=BMM_Additive
    scaleSpace=BCS_BoneSpace
    translationMode=BMM_Additive
    translationSpace=BCS_BoneSpace
    alphaBoolBlend {blendInTime=0.1, blendOption=Linear, blendOutTime=0.1}
    alphaScaleBiasClamp {bInterpResult=true}
    boneToModify {boneName=dyn_rearLeg_l_01}
    rotation=<rot 27.717 -30.695 7.779>
    translation=<-6.481 -6.076 2.307>
   }
   n45 = "Transform (Modify) Bone - Bone: dyn_frontLeg_l_01"!(ComponentPose=@n44, bAlphaBoolEnabled:bool=True, Translation="-0.874,2.781,1.076", Rotation="27.553,-21.829,-10.490", Scale="1,1,1") -> (Pose) {
    alphaInputType=Bool
    rotationMode=BMM_Additive
    rotationSpace=BCS_BoneSpace
    scaleMode=BMM_Additive
    scaleSpace=BCS_BoneSpace
    translationMode=BMM_Additive
    translationSpace=BCS_BoneSpace
    alphaBoolBlend {blendInTime=0.1, blendOption=Linear, blendOutTime=0.1}
    alphaScaleBiasClamp {bInterpResult=true}
    boneToModify {boneName=dyn_frontLeg_l_01}
    rotation=<rot 27.553 -21.829 -10.49>
    translation=<-0.874 2.781 1.076>
   }
   n46 = "Transform (Modify) Bone - Bone: dyn_frontLeg_l_02"!(ComponentPose=@n45, bAlphaBoolEnabled:bool=True, Translation="1.082,2.709,0.623", Rotation="18.064,-26.291,-6.114", Scale="1,1,1") -> (Pose) {
    alphaInputType=Bool
    rotationMode=BMM_Additive
    rotationSpace=BCS_BoneSpace
    scaleMode=BMM_Additive
    scaleSpace=BCS_BoneSpace
    translationMode=BMM_Additive
    translationSpace=BCS_BoneSpace
    alphaBoolBlend {blendInTime=0.1, blendOption=Linear, blendOutTime=0.1}
    alphaScaleBiasClamp {bInterpResult=true}
    boneToModify {boneName=dyn_frontLeg_l_02}
    rotation=<rot 18.064 -26.291 -6.114>
    translation=<1.082 2.709 0.623>
   }
   n47 = "Transform (Modify) Bone - Bone: dyn_outerLeg_l_01"(ComponentPose=@n46, bAlphaBoolEnabled=True, Translation="-2.096,-1.466,0.199", Rotation="1.852,-27.816,-17.852", Scale="1,1,1") {
    alphaInputType=Bool
    rotationMode=BMM_Additive
    rotationSpace=BCS_BoneSpace
    scaleMode=BMM_Additive
    scaleSpace=BCS_BoneSpace
    translationMode=BMM_Additive
    translationSpace=BCS_BoneSpace
    alphaBoolBlend {blendInTime=0.1, blendOption=Linear, blendOutTime=0.1}
    alphaScaleBiasClamp {bInterpResult=true}
    boneToModify {boneName=dyn_outerLeg_l_01}
    rotation=<rot 1.852 -27.816 -17.852>
    translation=<-2.096 -1.466 0.199>
   }
   n48 = Get MeshToCopy()
   n49 = RigidBody!(
    ComponentPose=Local To Component(LocalPose=Copy Pose From Mesh(SourceMeshComponent=@n48) {bCopyCurves=true, bUseAttachedParent=true})
    Alpha
    OverrideWorldGravity:Vector=Get GravityOverride()
    ExternalForce
   ) -> (Pose) {
    bClampLinearTranslationLimitToRefPose=true
    bOverrideWorldGravity=true
    overridePhysicsAsset=&/M_MED_Gumshoe_Cloth_Physics
    simulationSpace=BaseBoneSpace
    alphaBoolBlend {blendOption=Linear}
    baseBoneRef {boneName=root}
    componentAppliedLinearAccClamp=<5000 5000 5000>
    componentLinearAccScale=<0.5 0.5 0.5>
    overrideWorldGravity {z=-980}
    simSpaceSettings {
     maxAngularAcceleration=1000
     maxAngularVelocity=3
     maxLinearAcceleration=700
     maxLinearVelocity=500
     velocityScaleZ=0
     externalLinearDragV=<0.2 0.2>
    }
   }
   Output Pose(Result=Blend Poses by bool(BlendPose_0=Component To Local(ComponentPose=@n49), BlendPose_1=Component To Local(ComponentPose=@n47), bActiveValue=Get bIsInFrontEnd()) {blendType=Linear}) {layerGroup=None}
  ]}
 EventGraph {nodes [
   Event BlueprintUpdateAnimation()
   > Get!(execute:Exec) -> (then:Exec, else:Exec, Skeletal_Mesh:Skeletal Mesh Component Object Reference)
   > n1 = Sequence!(execute) -> (then_0, then_1, then_2:Exec, then_3:Exec)
   n2 = Gravity Override State Function DEPRECATED(
    execute=@n1.then_2
    CharacterPartAnimInstance=Self-Reference()
    GameTimeInSeconds=GetGameTimeInSeconds()
    GravityOverrideParams=Get Gravity Override Params()
   )
   > Set GravityOverride(GravityOverride=@n2)
   n3 = Branch(execute=@n1.then_3, Condition=OR Boolean(A=Get bIsJumping(), B=Get bIsFalling()))
   > Set NewJumping(NewJumping:bool=true) -> (Output_Get:bool)
   Set NewJumping(execute=@n3.else, NewJumping=false)
   n4 = Branch(execute=@n1.then_0, Condition="float > float"(A=Get PawnSpeed2D(), B=100) {is=title K2Node_CallFunction})
   n5 = Get Skeletal_Mesh() -> (Skeletal_Mesh:Skeletal Mesh Component Object Reference)
   n6 = GetDeltaTransformFromRefPose(self=@n5, BoneName=thigh_l, BaseName=pelvis)
   n7 = MapRangeClamped(Value=BreakRotator(InRot=BreakTransform(InTransform=@n6).Rotation).Yaw, InRangeA=-20, InRangeB=-100, OutRangeB=0.7)
   > Set thigh_rot_l(thigh_rot_l=float * float(A=@n7, B=0.25))
   n8 = GetDeltaTransformFromRefPose(self=@n5, BoneName=thigh_r, BaseName=pelvis)
   n9 = MapRangeClamped(Value=BreakRotator(InRot=BreakTransform(InTransform=@n8).Rotation).Yaw, InRangeA=-20, InRangeB=-100)
   > Set thigh_rot_r(thigh_rot_r=float * float(A=@n9, B=0.5))
   Set thigh_rot_l(execute=@n4.else, thigh_rot_l=@n7)
   > Set thigh_rot_r(thigh_rot_r=@n9)
   n10 = Get Skeletal_Mesh()
   n11 = BreakTransform(InTransform=GetDeltaTransformFromRefPose(self=@n10, BoneName=calf_l, BaseName=thigh_l))
   Set calf_rot_l(execute=@n1.then_1, calf_rot_l:double=MapRangeClamped(Value=BreakRotator(InRot=@n11.Rotation).Yaw, InRangeA=0, InRangeB=115)) -> (Output_Get:double)
   n12 = BreakTransform(InTransform=GetDeltaTransformFromRefPose(self=@n10, BoneName=thigh_r, BaseName=calf_r))
   > Set calf_rot_r(calf_rot_r:double=MapRangeClamped(Value=BreakRotator(InRot=@n12.Rotation).Yaw, InRangeA=0, InRangeB=-115)) -> (Output_Get:double)
   Event BlueprintInitializeAnimation()
   > Set MeshToCopy(MeshToCopy:Skeletal Mesh Component Object Reference=GetOwnerSkeletalMeshComponent()) -> (Output_Get:Skeletal Mesh Component Object Reference)
   > Set Skeletal_Mesh(Skeletal_Mesh:Skeletal Mesh Component Object Reference=GetOwnerSkeletalMeshComponent()) -> (Output_Get:Skeletal Mesh Component Object Reference)
  ]}
]
variables [
 calf_rot_l {type=number}
 calf_rot_r {type=number}
 Gravity Override Params {type=GravityOverrideParamsStruct}
 GravityOverride {type=Vector}
 MeshToCopy {type=SkeletalMeshComponent}
 NewJumping {type=boolean}
 Skeletal_Mesh {type=SkeletalMeshComponent}
 thigh_rot_l {type=number}
 thigh_rot_r {type=number}
]
